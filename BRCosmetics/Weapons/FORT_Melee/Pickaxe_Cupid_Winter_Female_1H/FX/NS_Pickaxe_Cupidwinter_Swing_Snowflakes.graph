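# Niagara swing effect for the Cupid Winter pickaxe that sheds small light-blue snowflake sprites as it moves.
# One emitter spawns by distance traveled during a short delayed window, sampling positions with random offsets on the skeletal mesh from User.Swing_SK_Mesh.
# Short-lived flakes get a slight push and inherited velocity, then drift down under drag and gentle gravity.
# A curve pops their size in and shrinks it away, with a small camera offset.
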NiagaraSystem /BRCosmetics/Weapons/FORT_Melee/Pickaxe_Cupid_Winter_Female_1H/FX/NS_Pickaxe_Cupidwinter_Swing_Snowflakes
dependencies {usedDataInterfaces [&NiagaraDataInterfaceCurve], usedDynamicInputs [], usedModules [&SystemState]}
emitters {SnowFlarkes {
  emitterUpdateScript [
   &EmitterState(
    Life Cycle Mode=::NewEnumerator1
    Loop Behavior=::NewEnumerator1
    Loop Delay=0.05
    Loop Duration=0.3
    Scalability Mode=::NewEnumerator1
    UseLoopDelay=true
   )
   &SpawnPerUnit(Module.Use Max Movement Threshold=true, Module.Use Movement Tolerance=false, Spawn Per Unit=40, Use Spawn Probability=false)
  ]
  particleSpawnScript [
   /Niagara/Modules/Spawn/Initialization/InitializeParticle(
    Color=<rgba 0.2456 0.5543 1 1>
    Lifetime=&RandomRangeFloat(Minimum=0.2, Maximum=0.5)
    Module.Write Color=true
    Module.Write Lifetime=true
    Module.Write Mass=true
    Module.Write Material Random=true
    Module.Write Position=false
    Module.Write Scale=false
    Module.Write SpriteRotation=false
    Module.Write SpriteSize=true
    Position=&SimulationPosition()
    Sprite Rotation=&RandomRangeFloat(Minimum=0, Maximum=360)
    Sprite Size=&Vector2DFromFloat(Value=&RandomRangeFloat(Minimum=8, Maximum=10))
   )
   &AddVelocityFromPoint(Module.Use Falloff Distance=false, Velocity Strength=10)
   &SkeletalMeshLocation(
    Inherit Velocity=false
    MaskSpawnGroup=false
    Mesh Sampling Type=::NewEnumerator5
    Sampled Position Offset=&RandomRangeVector(Minimum=<-10 -10 -10>, Maximum=<10 10 0>)
    Sampled Position Offset Coordinate Space:ENiagaraMeshTransforms=::NewEnumerator3
    Skeletal Mesh=$User.Swing_SK_Mesh
    Socket Sampling Mode=::NewEnumerator5
   ) {module=SkeletalMeshLocation001}
   &InheritVelocity(Module.Limit Speed=false, Velocity Scale=<0.01 0.01 0.01>)
   /SetVariables(Particles.Rotation:NiagaraFloat=&RandomRangeFloat(Minimum=-80, Maximum=80)) {isSetParameters=true}
  ]
  particleUpdateScript [
   &ParticleState()
   &Drag(Drag=4, Use Linear Drag=true, Use Rotational Drag=false)
   &GravityForce(Gravity=<0 0 -100>)
   &SolveForcesAndVelocity(Clamp Velocity=false, Limit Acceleration=false)
   &CameraOffset(Camera Offset Amount=5)
   &ScaleSpriteSize(Scale Factor=&Vector2DFromFloat(Value=&Multiply_Float(
      A=0.5
      B=&FloatFromCurve(FloatCurve=DataInterface() {propertyValues {Curve {keys [<at 0.0025 0.0142 RCIM_Cubic>, {arriveTangent=-0.0142, interpMode=RCIM_Cubic, leaveTangent=-0.0142, time=0.102, value=1.0065}, <at 1 0>]}}})
     )))
  ]
  rendererClasses [&NiagaraSpriteRendererProperties]
  renderers [{rendererClass=&NiagaraSpriteRendererProperties}]
 }}
summary {
 systemName=NS_Pickaxe_Cupidwinter_Swing_Snowflakes
 emitters [{emitterName=SnowFlarkes, rendererClasses [&NiagaraSpriteRendererProperties]}]
 userVariables [User.Swing_SK_Mesh {
   type=NiagaraDataInterfaceSkeletalMesh
   defaultValue {
    struct=NiagaraExt_VariableValue_DataInterface
    value {dataInterface=&self:NiagaraDataInterfaceSkeletalMesh_0}
   }
  }]
}
